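# FM_DoorOpen is a SoundCue for a door-opening sound in PilgrimQuickplay, built from portal effects and a crowd clip.
# A Mixer layers the Portal_unlocked and Portal_lowzap waves immediately, plus a much quieter crowd song-end clip that starts after a one-second Delay.
# It plays in the in-game SFX sound class, with attenuation and spatialization disabled so it is heard non-positionally.
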SoundCue /PilgrimQuickplay/Audio/FM_DoorOpen
root=Mixer_0
nodes [
 {id=Delay_0, children [WavePlayer_3], properties {delayMax=1, delayMin=1}}
 {id=Mixer_0, children [WavePlayer_0,WavePlayer_1,Delay_0], properties {inputVolume [1,1,0.1]}}
 {id=WavePlayer_0, properties {soundWaveAssetPtr=&/Portal_unlocked}}
 {id=WavePlayer_1, properties {soundWaveAssetPtr=&/Portal_lowzap}}
 {id=WavePlayer_3, properties {soundWaveAssetPtr=/SparksMaps/Audio/Wavs/crowd_lg_song_end2}}
]
properties {
 maxDistance=2097152
 soundClassObject=&Fort_SFX_InGame
 volumeMultiplier=1
 attenuationOverrides {bAttenuate=false, bSpatialize=false, falloffDistance=3000, attenuationShapeExtents {x=6000}}
}
