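# Sound cue for a door rumble effect in PilgrimQuickplay.
# A single wave player feeds the Portal_rumble sound straight to the output, with no randomization or mixing.
# It routes through the Fort_SFX_InGame sound class and references the shared FM_DoorOpen_atten attenuation asset for distance behavior.
# The cue also stores attenuation override values, but the override switch is left at its default.
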SoundCue /PilgrimQuickplay/Audio/FM_DoorRumble
root=WavePlayer_7
nodes [{id=WavePlayer_7, properties {soundWaveAssetPtr=&/Portal_rumble}}]
properties {
 attenuationSettings=&/FM_DoorOpen_atten
 maxDistance=6000
 soundClassObject=&Fort_SFX_InGame
 volumeMultiplier=1
 attenuationOverrides {bAttenuate=false, bSpatialize=false, falloffDistance=3000, attenuationShapeExtents {x=6000}}
}
