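# Sound cue for a looping organ ambience in Pilgrim Quickplay.
# A single looping wave player feeds the root with glitch_organ_ambience_2, despite the cue's "4" suffix.
# It plays at reduced volume, is spatialized through the shared organ ambience attenuation asset, and routes to the Sparks GreenRoom sound class.
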SoundCue /PilgrimQuickplay/Audio/organ_ambience_4_Cue
root=WavePlayer_0
nodes [{id=WavePlayer_0, properties {bLooping=true, soundWaveAssetPtr=&/glitch_organ_ambience_2}}]
properties {
 attenuationSettings=&/organ_ambience_attenuation
 maxDistance=1700
 soundClassObject=/SparksMaps/Audio/GreenRoom/Sparks_GreenRm
 volumeMultiplier=0.5
}
