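; PilgrimQuickplayUIDirector is a dynamic UI director for Pilgrim Quickplay.
; On BeginPlay it gets the FortPickerContext subsystem from the owning local player controller and sets its picker data to the shared in-game HUD FortPickers asset.
; The construction script is empty.
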(event EventBeginPlay
  (Setup|SetFortPickerData (PlayerController|LocalPlayerSubsystems|GetFortPickerContext (DynamicUIDirector|GetOwningLocalPlayerController)) "/Game/UI/InGame/HUD/FortPickers.FortPickers"))

(fn ConstructionScript)
